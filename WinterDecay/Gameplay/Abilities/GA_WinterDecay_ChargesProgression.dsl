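; GA_WinterDecay_ChargesProgression is a gameplay ability that applies the WinterDecay charges progression gameplay effect to its owner for as long as the ability is active.
; On activation it applies GE_WinterDecay_ChargesProgression and stores the returned handle in the AmmoEffect variable.
; On end it removes the effect through that stored handle.
; Both events print on-screen debug strings, "Equipped" and "Removed", which suggest an equip-driven lifetime.
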(event Ability|EventActivateAbility
  (Development|PrintString "Equipped" true true "(R=1.000000,G=0.290341,B=0.169553,A=1.000000)" 16.0)
  (bind _returnvalue (Ability|ApplyGameplayEffectToOwner "/WinterDecay/GameplayEffects/GE_WinterDecay_ChargesProgression.GE_WinterDecay_ChargesProgression_C"))
  (Variables|Default|SetAmmoEffect _returnvalue))

(event Ability|EventOnEndAbility
  (params (bWasCancelled :type Boolean))
  (Ability|RemoveGameplayEffectFromOwnerWithHandle (Variables|Default|GetAmmoEffect))
  (Development|PrintString "Removed" true true "(R=1.000000,G=0.290341,B=0.169553,A=1.000000)" 16.0))
